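# Physics asset for the wing mesh of the HornedJudgment cape cosmetic, defining capsule bodies and joints for its physics.
# Left and right wing capsules attach to a spine pivot with tight, matching swing and twist limits, allowing only slight movement.
# The spine pivot and a small cape_top body each connect to a hider body through constraints with free swing and twist.
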PhysicsAsset /BRCosmetics/Accessories/FORT_Capes/Cape_M_MED_HornedJudgment/Meshes/M_MED_HornedJudgment_Wing_Physics
bodies {
 Spinepivot_01_jnt [{shapeType=Capsule, length=5, radius=1, shapeName=Spinepivot_01_jnt_capsule, center=<2.4638 1.4857 0>, rotation=<rot 0 -90 0>}]
 Wing_01_l_jnt [{shapeType=Capsule, length=8, radius=4, shapeName=Wing_01_l_jnt_capsule, center=<-12 0 0>, rotation=<rot 90 0 0>}]
 Wing_01_r_jnt [{shapeType=Capsule, length=8, radius=4, shapeName=Wing_01_r_jnt_capsule, center=<12 0 0>, rotation=<rot 90 0 0>}]
 cape_top [{shapeType=Capsule, length=1, radius=1, shapeName=cape_top_capsule}]
 hider [{shapeType=Capsule, length=5, radius=1, shapeName=hider_capsule, center=<0.0017 -0.1219 -0.4656>, rotation=<rot 0 -89.9851 -89.9864>}]
}
constraints [
 {bone1Name=Wing_01_l_jnt, bone2Name=Spinepivot_01_jnt, swing1LimitDegrees=20, swing2LimitDegrees=10, twistLimitDegrees=2}
 {bone1Name=Wing_01_r_jnt, bone2Name=Spinepivot_01_jnt, swing1LimitDegrees=20, swing2LimitDegrees=10, twistLimitDegrees=2}
 {
  bone1Name=cape_top
  bone2Name=hider
  swing1LimitDegrees=45
  swing1Motion=Free
  swing2LimitDegrees=45
  swing2Motion=Free
  twistLimitDegrees=45
  twistMotion=Free
 }
 {
  bone1Name=Spinepivot_01_jnt
  bone2Name=hider
  swing1LimitDegrees=45
  swing1Motion=Free
  swing2LimitDegrees=45
  swing2Motion=Free
  twistLimitDegrees=45
  twistMotion=Free
 }
]
